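# Material function that outputs a flat-shading color chosen by time of day, for materials in the WildEstate project.
# It exposes four vector parameters in the FlatShading group for morning, day, evening and night, defaulting to neutral greys except a blue night tint.
# Their RGB values feed the Color_Time_of_Day function, whose result becomes the single Color output.
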MaterialFunction /WildEstate/TechArt/MaterialFunctions/MF_TODColorParams
graph {nodes [FunctionOutput(None=MaterialFunctionCall(
    Color_Morning=VectorParameter() {group=FlatShading, parameterName=Color.ToD.Morning, defaultValue=<rgba 0.526 0.526 0.526 1>}.RGB
    Color_Day=VectorParameter() {group=FlatShading, parameterName=Color.ToD.Day, defaultValue=<rgba 0.7344 0.7344 0.7344 1>}.RGB
    Color_Evening=VectorParameter() {group=FlatShading, parameterName=Color.ToD.Evening, defaultValue=<rgba 0.4948 0.4948 0.4948 1>}.RGB
    Color_Night=VectorParameter() {group=FlatShading, parameterName=Color.ToD.Night, defaultValue=<rgba 0.1129 0.1433 0.5938 1>}.RGB
   ) {materialFunction=&Color_Time_of_Day}.Result) {outputName=Color}]}
