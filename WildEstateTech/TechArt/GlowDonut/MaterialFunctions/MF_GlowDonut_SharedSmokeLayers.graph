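# Shared material function for GlowDonut smoke layers, outputting color, sun and near-camera masks, gradient fades, distance blend and world position offset.
# Color blends toward time-of-day tint, with fresnel or sun-direction shading and darkening by height within instance bounds.
# WPO uses height-based sine bands with per-actor phase from pivot location, gradient-masked and scaled by object scale.
# A sum adding world-aligned noise displacement and wind sway is left unconnected.
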MaterialFunction /WildEstateTech/TechArt/GlowDonut/MaterialFunctions/MF_GlowDonut_SharedSmokeLayers
graph {nodes [
  n1 = TextureObject() {texture=/Game/Global/Textures/Default/Noise/3xPerlin_M}
  n2 = Time()
  n3 = VectorParameter() {parameterName=Speeds, defaultValue=<rgba 1 1 1 1>}
  n4 = MaterialFunctionCall(
   TextureObject=@n1
   TextureSize=ScalarParameter() {defaultValue=64, parameterName=Tiling}
   WorldPosition=Add(A=WorldPosition().XYZ, B=Multiply(A=@n2, B=@n3.RGB))
  ) {materialFunction=&WorldAlignedTexture}
  n5 = Multiply(A=Multiply(A=Subtract(A=@n4.XY Texture) {constB=0.5}) {constB=2}, B=ScalarParameter() {defaultValue=5, parameterName=Intensity})
  n6 = VertexNormalWS()
  n7 = Multiply(A=Time(), B=ScalarParameter() {defaultValue=0.1, parameterName=WPO.WindSpeed})
  n8 = LinearInterpolate(
   A=VectorParameter() {parameterName=DirA, defaultValue=<rgba 100 0 0 1>}.RGB
   B=VectorParameter() {parameterName=DirB, defaultValue=<rgba 100 0 0 1>}.RGB
   Alpha=ConstantBiasScale(None=Sine(None=@n7))
  )
  n9 = Distance(A=TransformPosition(None=WorldPosition().XYZ) {transformSourceType=TRANSFORMPOSSOURCE_World}, B=Constant())
  n10 = ComponentMask(None=Bounds().Max) {b=true}
  n11 = Power(Base=Saturate(None=Divide(A=@n9, B=@n10)), Exp=ScalarParameter() {defaultValue=1, parameterName=WPO.MaskCurve}) {constExponent=1}
  n12 = Divide(
   A=MaterialFunctionCall() {materialFunction=/Game/Effects/Niagara/MasterMaterials/Functions/InstanceBounds_0-1_UVW}.B
   B=ScalarParameter() {defaultValue=100, parameterName=WPO.PanningBands.Distant}
  ) {constB=100}
  n13 = Multiply(A=Time(), B=ScalarParameter() {defaultValue=0.1, parameterName=WPO.PanningBands.Speed})
  n14 = MaterialFunctionCall() {materialFunction=&ObjectPivotPoint}
  n15 = Sine(None=Add(A=Add(A=@n12, B=@n13), B=Length(None=@n14.Object Pivot Location) {is=class MaterialExpressionLength}) {desc=TIme Offset per Actor})
  n16 = ScalarParameter() {defaultValue=1, parameterName=WPO.PanningBands}
  n17 = Multiply(A=Multiply(A=Multiply(A=@n15, B=@n16), B=Constant3Vector() {constant=<rgba 1 1 0 0>}), B=NamedRerouteUsage())
  Add(A=Add(A=Multiply(A=@n5, B=@n6), B=LinearInterpolate(B=@n8, Alpha=@n11)), B=@n17)
  n18 = CameraPositionWS()
  ComponentMask(None=@n18) {g=true, r=true}
  ComponentMask(None=MaterialFunctionCall() {materialFunction=&ObjectPivotPoint}.Object Pivot Location) {g=true, r=true}
  n19 = VectorParameter() {parameterName=Color, channelNames {a=FresnelDarken, b="", g="", r=""}, defaultValue=<rgba 0.4125 0.5029 0.4397 0.9>}
  n20 = LinearInterpolate(
   A=@n19.RGB
   B=MaterialFunctionCall() {materialFunction=/WildEstate/TechArt/MaterialFunctions/MF_TODColorParams}.Color
   Alpha=ScalarParameter() {defaultValue=0.1, parameterName=Color.TOD.Influence}
  )
  n21 = Multiply(A=@n20, B=@n19.FresnelDarken) {constB=0.2}
  n22 = LinearInterpolate(
   A=MaterialFunctionCall(
    Brightness=ScalarParameter() {defaultValue=1, parameterName=Color.Fresnel.Brightness}
    Exponent=ScalarParameter() {defaultValue=1, parameterName=Color.Fresnel.Exp}
   ) {materialFunction=&MF_Fresnel_FX}.Fresnel
   B=NamedRerouteUsage()
   Alpha=ScalarParameter() {parameterName=Color.UseSunDir}
  )
  n23 = ScalarParameter() {defaultValue=0.6, parameterName=Color.LowerShading.Val}
  n24 = Power(
   Base=MaterialFunctionCall() {materialFunction=&InstanceBounds_0-1_UVW}.B
   Exp=ScalarParameter() {defaultValue=1, parameterName=Color.LowerShading.Power}
  )
  n25 = ScalarParameter() {defaultValue=1, parameterName=Color.LowerShading.Mult}
  n26 = Add(A=Saturate(None=Multiply(A=@n24, B=@n25)), B=ScalarParameter() {parameterName=Color.LowerShading.Offset})
  FunctionOutput(None=Multiply(A=LinearInterpolate(A=@n21, B=@n20, Alpha=@n22) {constAlpha=1, desc=Shadow Mask}, B=LinearInterpolate(A=@n23, Alpha=@n26))) {outputName=Color, sortPriority=1}
  FunctionOutput(None=NamedRerouteUsage()) {outputName=SunDirection, sortPriority=2}
  FunctionOutput(None=NamedRerouteUsage()) {outputName=NearCameraMask, sortPriority=3}
  FunctionOutput(None=Multiply(
    A=NamedRerouteUsage()
    B=MaterialFunctionCall() {materialFunction=&ObjectScale}.Scale X
   )) {outputName=WPO, sortPriority=4}
  FunctionOutput(None=NamedRerouteUsage()) {outputName=TopFade, sortPriority=8}
  FunctionOutput(None=NamedRerouteUsage()) {outputName=BottomFade, sortPriority=9}
  FunctionOutput(None=NamedRerouteUsage()) {outputName=MergedFade, sortPriority=7}
  FunctionOutput(None=NamedRerouteUsage()) {outputName=DistanceBlend, sortPriority=10}
  Length(None=@n14.Object Pivot Location)
  n27 = ScalarParameter() {parameterName=Sun.Mask.Min}
  n28 = ScalarParameter() {defaultValue=1, parameterName=Sun.Mask.Max}
  n29 = ConstantBiasScale(None=DotProduct(A=SkyAtmosphereLightDirection(), B=VertexNormalWS()))
  NamedRerouteDeclaration(None=Saturate(None=SmoothStep(Min=@n27, Max=@n28, Value=@n29))) {name=SunDirecitonAlpha, nodeColor=<rgba 1 0 0.7734 1>}
  NamedRerouteDeclaration(None=@n4.XY Texture) {name=WS_Warp, nodeColor=<rgba 1 0.7266 0 1>}
  n30 = Saturate(None=OneMinus(None=MaterialFunctionCall() {materialFunction=&LinearGradient}.VGradient))
  n31 = ScalarParameter() {defaultValue=1, parameterName=Fade.Bottom.Contrast}
  n32 = Saturate(None=Multiply(A=Power(Base=@n30, Exp=@n31), B=ScalarParameter() {defaultValue=1.5, parameterName=Fade.Bottom.Ctrl}) {constB=1.5})
  NamedRerouteDeclaration(None=@n32) {name=BottomFade, nodeColor=<rgba 0 1 0.9062 1>}
  NamedRerouteDeclaration(None=@n1) {name=TextureObj, nodeColor=<rgba 0 0.5078 1 1>}
  n33 = OneMinus(None=MaterialFunctionCall() {materialFunction=&LinearGradient}.VGradient)
  NamedRerouteDeclaration(None=LinearInterpolate(B=@n17, Alpha=Saturate(None=@n33))) {name=WPO, nodeColor=<rgba 1 0 0 1>}
  n34 = MaterialFunctionCall() {materialFunction=&LinearGradient}
  n35 = ScalarParameter() {defaultValue=1.5, parameterName=Fade.Top.Ctrl}
  n36 = NamedRerouteDeclaration(None=Saturate(None=Multiply(A=@n34.VGradient, B=@n35) {constB=1.5})) {name=TopFade, nodeColor=<rgba 0.5234 0 1 1>}
  NamedRerouteDeclaration(None=LinearInterpolate(A=@n36, B=Multiply(A=@n36, B=@n32), Alpha=ScalarParameter() {parameterName=Fade.Bottom.Blend})) {name=TopBottomFade, nodeColor=<rgba 1 0.8672 0 1>}
  n37 = WorldPosition()
  n38 = Divide(
   A=Add(A=Distance(A=@n18, B=@n37.XYZ), B=ScalarParameter() {group=Detail Blend, parameterName=DistanceBlend.Offset})
   B=ScalarParameter() {defaultValue=256, group=Detail Blend, parameterName=DistanceBlend.MaxDistance}
  )
  NamedRerouteDeclaration(None=Power(Base=Saturate(None=@n38), Exp=ScalarParameter() {defaultValue=1, group=Detail Blend, parameterName=DistanceBlend.Pwr})) {name=DistanceBlend, desc=0 far - 1 close, nodeColor=<rgba 0.3359 1 0 1>}
  n39 = CameraPositionWS()
  n40 = WorldPosition()
  n41 = ScalarParameter() {defaultValue=256, parameterName=RevealMaskRadius}
  NamedRerouteDeclaration(None=OneMinus(None=SphereMask(A=@n39, B=@n40.XYZ, Radius=@n41) {hardnessPercent=0.5})) {name=NearCameraMask, nodeColor=<rgba 1 0 0.4219 1>}
 ]}
